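# Physics asset for the GrowMallet backpack cosmetic, giving its tail and fins bodies for constrained secondary motion.
# A box covers the bowl, and a capsule chain runs from a zero-size base_tail anchor through four tail bones, branching into two-bone left and right fins.
# Narrow swing limits with zero twist restrict the tail and fins, while the base_tail-to-bowl joint leaves swing and twist free.
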PhysicsAsset /BRCosmetics/Accessories/FORT_Backpacks/Backpack_F_MED_GrowMallet/Meshes/F_MED_GrowMallet_Pack_Physics
bodies {
 base_bowl [{
   shapeType=Box
   extentX=0.7075
   extentY=5.1504
   extentZ=8.7852
   shapeName=base_bowl_box
   center=<5.2756 0.1957 2.3556>
   rotation=<rot -15.5319 -0.0155 -89.9435>
  }]
 base_tail [{shapeType=Capsule, length=0, radius=0, shapeName=base_tail_capsule, center=<1 0 0>, rotation=<rot 90 0 0>}]
 dyn_tail_01 [{shapeType=Capsule, length=1, radius=0.85, shapeName=dyn_tail_01_capsule, center=<1 0 0>, rotation=<rot 90 0 0>}]
 dyn_tail_02 [{shapeType=Capsule, length=1.2424, radius=0.6076, shapeName=dyn_tail_02_capsule, center=<1 0 0>, rotation=<rot 90 0 0>}]
 dyn_tail_03 [{shapeType=Capsule, length=1.2601, radius=0.5899, shapeName=dyn_tail_03_capsule, center=<1 0 0>, rotation=<rot 90 0 0>}]
 dyn_tail_04 [{shapeType=Capsule, length=1.0984, radius=0.7516, shapeName=dyn_tail_04_capsule, center=<1 0 0>, rotation=<rot 90 0 0>}]
 dyn_tailfin_l_01 [{shapeType=Capsule, length=1, radius=1, shapeName=dyn_tailfin_l_01_capsule, center=<1 0 0>, rotation=<rot 90 0 0>}]
 dyn_tailfin_l_02 [{shapeType=Capsule, length=1, radius=1, shapeName=dyn_tailfin_l_02_capsule, center=<1 0 0>, rotation=<rot 90 0 0>}]
 dyn_tailfin_r_01 [{shapeType=Capsule, length=1, radius=1, shapeName=dyn_tailfin_r_01_capsule, center=<-1 0 0>, rotation=<rot 90 0 0>}]
 dyn_tailfin_r_02 [{shapeType=Capsule, length=1, radius=1, shapeName=dyn_tailfin_r_02_capsule, center=<-1 0 0>, rotation=<rot 90 0 0>}]
}
constraints [
 {bone1Name=dyn_tail_01, bone2Name=base_tail, swing1LimitDegrees=0, swing2LimitDegrees=8, twistLimitDegrees=0}
 {bone1Name=dyn_tail_02, bone2Name=dyn_tail_01, swing1LimitDegrees=5, swing2LimitDegrees=15, twistLimitDegrees=0}
 {bone1Name=dyn_tail_03, bone2Name=dyn_tail_02, swing1LimitDegrees=5, swing2LimitDegrees=15, twistLimitDegrees=0}
 {bone1Name=dyn_tail_04, bone2Name=dyn_tail_03, swing1LimitDegrees=5, swing2LimitDegrees=15, twistLimitDegrees=0}
 {bone1Name=dyn_tailfin_l_01, bone2Name=dyn_tail_04, swing1LimitDegrees=5, swing2LimitDegrees=15, twistLimitDegrees=0}
 {bone1Name=dyn_tailfin_l_02, bone2Name=dyn_tailfin_l_01, swing1LimitDegrees=5, swing2LimitDegrees=15, twistLimitDegrees=0}
 {bone1Name=dyn_tailfin_r_01, bone2Name=dyn_tail_04, swing1LimitDegrees=5, swing2LimitDegrees=15, twistLimitDegrees=0}
 {bone1Name=dyn_tailfin_r_02, bone2Name=dyn_tailfin_r_01, swing1LimitDegrees=5, swing2LimitDegrees=15, twistLimitDegrees=0}
 {
  bone1Name=base_tail
  bone2Name=base_bowl
  swing1LimitDegrees=45
  swing1Motion=Free
  swing2LimitDegrees=45
  swing2Motion=Free
  twistLimitDegrees=45
  twistMotion=Free
 }
]
