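# Animation Blueprint for the GrowMallet back bling, animating its bowl, tail and reactions to player events.
# Additive state machines layer idle breaks and reactions over a keep-alive loop, adding crouch, jump, glide and DBNO responses outside the frontend.
# In gameplay, bowl rotation follows DBNO, while tail bones blend RigidBody physics with procedural offsets and inertialization.
# The event graph feeds input events, player motion, idle-break timing and eye-material change flags.
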AnimBlueprint /BRCosmetics/Accessories/FORT_Backpacks/Backpack_F_MED_GrowMallet/Meshes/F_MED_GrowMallet_Pack_AnimBP
target_skeleton=&/F_MED_GrowMallet_Pack_Skeleton
graphs [
 AnimGraph {nodes [
   n1 = Layered blend per bone(BasePose=Use cached pose 'RBAN'(), BlendPoses_0=Use cached pose 'Adjustments'()) {
    bMeshSpaceRotationBlend=true
    bRootSpaceRotationBlend=true
    bUpdateBasePoseFirst=true
    curveBlendOption=NormalizeByWeight
    layerSetup [{branchFilters [{blendDepth=8, boneName=base_tail}]}]
   }
   n2 = Use cached pose 'AnimationCache'()
   n3 = Get bShouldDisplayFrontendReactivity() -> (bShouldDisplayFrontendReactivity:bool)
   Output Pose(Result=Inertialization!(Source:Pose Link Structure=Blend Poses by bool(BlendPose_0=@n2, BlendPose_1=@n1, bActiveValue=@n3)) -> (Pose:Pose Link Structure) {is=title AnimGraphNode_Inertialization, bResetOnBecomingRelevant=true}) {layerGroup=None}
   AnimGraphNode_StateMachine_0 = FE Idle/Reaction() {is=title AnimGraphNode_StateMachine}
   AnimGraphNode_StateMachine_1 = Idle/reaction() {is=title AnimGraphNode_StateMachine}
   n4 = Blend Poses by bool(BlendPose_0=@AnimGraphNode_StateMachine_0, BlendPose_1=@AnimGraphNode_StateMachine_1, bActiveValue=Get bShouldDisplayFrontendReactivity())
   n5 = Sequence Player 'GrowMallet_KeepAlive_Anim_PACK'() {sequence=/BRCosmetics/Accessories/FORT_Backpacks/Backpack_F_MED_GrowMallet/Animations/GrowMallet_KeepAlive_Anim_PACK}
   AnimGraphNode_StateMachine_2 = Reactive State Machine() {is=title AnimGraphNode_StateMachine}
   n6 = Get bShouldDisplayFrontendReactivity()
   Save cached pose 'AnimationCache'(Pose=Apply Additive!(
     Base=Apply Additive!(Base=@n5, Additive=@n4, Alpha:float) -> (Pose)
     Additive=@AnimGraphNode_StateMachine_2
     bAlphaBoolEnabled:bool=NOT Boolean(A=@n6)
    ) -> (Pose) {alphaInputType=Bool}) {cacheName=AnimationCache}
   n7 = RigidBody!(ComponentPose=Local To Component(LocalPose=Use cached pose 'RBAN'()), Alpha, OverrideWorldGravity:Vector="0,0,-980", ExternalForce) -> (Pose) {
    bClampLinearTranslationLimitToRefPose=true
    bOverrideWorldGravity=true
    bUseDefaultAsSimulated=true
    overridePhysicsAsset=&/F_MED_GrowMallet_Pack_Physics
    alphaScaleBiasClamp {interpSpeedDecreasing=5, interpSpeedIncreasing=2}
    baseBoneRef {boneName=base_tail}
    simSpaceSettings {
     dampingAlpha=0
     maxAngularAcceleration=10
     maxAngularVelocity=5
     maxLinearAcceleration=1200
     maxLinearVelocity=700
     worldAlpha=0.85
     externalLinearDragV=<1 1 1>
    }
   }
   Save cached pose 'Physics'(Pose=Component To Local(ComponentPose=@n7)) {cacheName=Physics}
   n8 = Property Access!() -> (Value_Location:Vector, Value_Rotation:Rotator, Value_Scale:Vector) {is=title K2Node_PropertyAccess}
   n9 = Property Access()
   n10 = "Transform (Modify) Bone - Bone: base_tail"(
    ComponentPose=Local To Component(LocalPose=Use cached pose 'RBAN'())
    Alpha=1
    Translation=@n9
    Rotation=@n9.Value_Rotation
    Scale=@n9.Value_Scale
   ) {
    rotationMode=BMM_Additive
    scaleMode=BMM_Additive
    translationMode=BMM_Additive
    alphaBoolBlend {blendOption=Linear}
    alphaScaleBiasClamp {bInterpResult=true}
    boneToModify {boneName=base_tail}
   }
   n11 = Property Access()
   n12 = "Transform (Modify) Bone - Bone: dyn_tail_01"(ComponentPose=@n10, Alpha=1, Translation=@n11, Rotation=@n11.Value_Rotation, Scale=Get ScaleOffset() -> (ScaleOffset:Vector)) {
    rotationMode=BMM_Additive
    scaleMode=BMM_Additive
    translationMode=BMM_Additive
    alphaBoolBlend {blendOption=Linear}
    alphaScaleBiasClamp {bInterpResult=true}
    boneToModify {boneName=dyn_tail_01}
   }
   n13 = Property Access()
   n14 = "Transform (Modify) Bone - Bone: dyn_tail_02"(ComponentPose=@n12, Alpha=1, Translation=@n13, Rotation=@n13.Value_Rotation, Scale=Get ScaleOffset()) {
    rotationMode=BMM_Additive
    scaleMode=BMM_Additive
    translationMode=BMM_Additive
    alphaBoolBlend {blendOption=Linear}
    alphaScaleBiasClamp {bInterpResult=true}
    boneToModify {boneName=dyn_tail_02}
   }
   n15 = Property Access()
   n16 = "Transform (Modify) Bone - Bone: dyn_tail_03"(ComponentPose=@n14, Alpha=1, Translation=@n15, Rotation=@n15.Value_Rotation, Scale=Get ScaleOffset()) {
    rotationMode=BMM_Additive
    scaleMode=BMM_Additive
    translationMode=BMM_Additive
    alphaBoolBlend {blendOption=Linear}
    alphaScaleBiasClamp {bInterpResult=true}
    boneToModify {boneName=dyn_tail_03}
   }
   n17 = Property Access()
   n18 = "Transform (Modify) Bone - Bone: dyn_tail_04"(ComponentPose=@n16, Alpha=1, Translation=@n17, Rotation=@n17.Value_Rotation, Scale=Get ScaleOffset()) {
    rotationMode=BMM_Additive
    scaleMode=BMM_Additive
    translationMode=BMM_Additive
    alphaBoolBlend {blendOption=Linear}
    alphaScaleBiasClamp {bInterpResult=true}
    boneToModify {boneName=dyn_tail_04}
   }
   n19 = "Transform (Modify) Bone - Bone: dyn_tailfin_l_01"(ComponentPose=@n18, Alpha=1, Translation=@n8, Rotation=@n8.Value_Rotation, Scale=Get ScaleOffset()) {
    rotationMode=BMM_Additive
    scaleMode=BMM_Additive
    translationMode=BMM_Additive
    alphaBoolBlend {blendOption=Linear}
    alphaScaleBiasClamp {bInterpResult=true}
    boneToModify {boneName=dyn_tailfin_l_01}
   }
   n20 = Property Access()
   n21 = "Transform (Modify) Bone - Bone: dyn_tailfin_l_02"(ComponentPose=@n19, Alpha=1, Translation=@n20, Rotation=@n20.Value_Rotation, Scale=Get ScaleOffset()) {
    rotationMode=BMM_Additive
    scaleMode=BMM_Additive
    translationMode=BMM_Additive
    alphaBoolBlend {blendOption=Linear}
    alphaScaleBiasClamp {bInterpResult=true}
    boneToModify {boneName=dyn_tailfin_l_02}
   }
   n22 = Property Access()
   n23 = "Transform (Modify) Bone - Bone: dyn_tailfin_r_01"(ComponentPose=@n21, Alpha=1, Translation=@n22, Rotation=@n22.Value_Rotation, Scale=Get ScaleOffset()) {
    rotationMode=BMM_Additive
    scaleMode=BMM_Additive
    translationMode=BMM_Additive
    alphaBoolBlend {blendOption=Linear}
    alphaScaleBiasClamp {bInterpResult=true}
    boneToModify {boneName=dyn_tailfin_r_01}
   }
   n24 = Property Access()
   n25 = "Transform (Modify) Bone - Bone: dyn_tailfin_r_02"(ComponentPose=@n23, Alpha=1, Translation=@n24, Rotation=@n24.Value_Rotation, Scale=Get ScaleOffset()) {
    rotationMode=BMM_Additive
    scaleMode=BMM_Additive
    translationMode=BMM_Additive
    alphaBoolBlend {blendOption=Linear}
    alphaScaleBiasClamp {bInterpResult=true}
    boneToModify {boneName=dyn_tailfin_r_02}
   }
   n26 = Layered blend per bone(BasePose=Component To Local(ComponentPose=@n25), BlendPoses_0=Use cached pose 'Physics'()) {
    bMeshSpaceRotationBlend=true
    bRootSpaceRotationBlend=true
    curveBlendOption=NormalizeByWeight
    layerSetup [{branchFilters [{blendDepth=8, boneName=base_tail}]}]
   }
   Save cached pose 'Adjustments'(Pose=Inertialization(Source=@n26)) {cacheName=Adjustments}
   n27 = "Transform (Modify) Bone - Bone: base_bowl"(
    ComponentPose=Local To Component(LocalPose=Use cached pose 'AnimationCache'())
    Alpha=1
    Translation="0,0,0"
    Rotation="-90,0,0"
    Scale="1,1,1"
   ) {
    rotationMode=BMM_Additive
    rotationSpace=BCS_BoneSpace
    alphaBoolBlend {blendOption=Linear}
    boneToModify {boneName=base_bowl}
    rotation {pitch=-90}
   }
   n28 = "Transform (Modify) Bone - Bone: base_bowl"(
    ComponentPose=Local To Component(LocalPose=Use cached pose 'AnimationCache'())
    Alpha=1
    Translation="0,0,0"
    Rotation="0,0,0"
    Scale="1,1,1"
   ) {
    rotationMode=BMM_Additive
    rotationSpace=BCS_BoneSpace
    alphaBoolBlend {blendOption=Linear}
    alphaScaleBiasClamp {bInterpResult=true}
    boneToModify {boneName=base_bowl}
   }
   n29 = "Transform (Modify) Bone - Bone: base_bowl"!(ComponentPose=@n28, bAlphaBoolEnabled:bool=True, Translation="0,0,0", Rotation="0,0,0", Scale="1,1,1") -> (Pose) {
    alphaInputType=Bool
    rotationMode=BMM_Additive
    rotationSpace=BCS_BoneSpace
    alphaBoolBlend {blendOption=QuarticInOut, blendOutTime=10}
    alphaScaleBiasClamp {bInterpResult=true}
    boneToModify {boneName=base_bowl}
   }
   Save cached pose 'RBAN'(Pose=Blend Poses by bool(BlendPose_0=Component To Local(ComponentPose=@n27), BlendPose_1=Component To Local(ComponentPose=@n29), bActiveValue=Get bIsDBNO()) {blendType=Linear}) {cacheName=RBAN}
  ]}
 FE Idle/Reaction {
  entry=FE Idle
  path=AnimGraph.AnimGraphNode_StateMachine_0.FE Idle/Reaction
  ids=id
  nodes [
   To FE Idle() {is=title AnimStateAliasNode, also {aliases [FE Elims/OpenChest]}}
   To FE Elims/OpenChest() {is=title AnimStateAliasNode, also {aliases [FE Idle, FE Idle Break]}}
   AnimStateNode_0 = FE Idle() {
    is=title AnimStateNode
    bAlwaysResetOnEntry=true
    stateEntered {guid=#1}
    stateFullyBlended {guid=#2}
    stateLeft {guid=#3}
   }
   AnimStateNode_1 = FE Elims/OpenChest() {is=title AnimStateNode}
   AnimStateNode_5 = FE Idle Break() {is=title AnimStateNode}
   AnimStateTransitionNode_10 = AnimStateTransitionNode() {bAllowInertializationForSelfTransitions=true, bAutomaticRuleBasedOnSequencePlayerInState=true}
   AnimStateTransitionNode_13 = AnimStateTransitionNode() {bAllowInertializationForSelfTransitions=true, bAutomaticRuleBasedOnSequencePlayerInState=true}
   AnimStateTransitionNode_3 = AnimStateTransitionNode() {bAllowInertializationForSelfTransitions=true}
   AnimStateTransitionNode_6 = AnimStateTransitionNode() {bAllowInertializationForSelfTransitions=true}
  ]
  links [
   {
    dst=FE Idle Break
    rule=AnimGraph.AnimGraphNode_StateMachine_0.FE Idle/Reaction.AnimStateTransitionNode_6.AnimationTransitionGraph_0
    src=FE Idle
    via=AnimStateTransitionNode_6
   }
   {
    dst=FE Idle
    rule=AnimGraph.AnimGraphNode_StateMachine_0.FE Idle/Reaction.AnimStateTransitionNode_13.Transition
    src=FE Idle Break
    via=AnimStateTransitionNode_10
   }
   {
    dst=FE Elims/OpenChest
    rule=AnimGraph.AnimGraphNode_StateMachine_0.FE Idle/Reaction.AnimStateTransitionNode_3.AnimationTransitionGraph_1
    src=To FE Elims/OpenChest
    via=AnimStateTransitionNode_3
   }
   {
    dst=FE Idle
    rule=AnimGraph.AnimGraphNode_StateMachine_0.FE Idle/Reaction.AnimStateTransitionNode_13.Transition
    src=To FE Idle
    via=AnimStateTransitionNode_13
   }
  ]
 }
 FE Idle {
  path=AnimGraph.AnimGraphNode_StateMachine_0.FE Idle/Reaction.AnimStateNode_0.FE Idle
  nodes [Output Animation Pose(Result="Sequence Player (additive) 'GrowMallet_Idle_Anim_PACK'"() {sequence=/BRCosmetics/Accessories/FORT_Backpacks/Backpack_F_MED_GrowMallet/Animations/GrowMallet_Idle_Anim_PACK})]
 }
 FE Elims/OpenChest {
  path=AnimGraph.AnimGraphNode_StateMachine_0.FE Idle/Reaction.AnimStateNode_1.FE Elims/OpenChest
  nodes [
   n1 = "Sequence Player (additive) 'GrowMallet_Elimination_FE_Anim_PACK'"() {
    bLoopAnimation=false
    sequence=/BRCosmetics/Accessories/FORT_Backpacks/Backpack_F_MED_GrowMallet/Animations/GrowMallet_Elimination_FE_Anim_PACK
   }
   Output Animation Pose(Result=@n1)
  ]
 }
 FE Idle Break {
  path=AnimGraph.AnimGraphNode_StateMachine_0.FE Idle/Reaction.AnimStateNode_5.FE Idle Break
  nodes [
   n1 = Random Sequence Player!() -> (Pose:Pose Link Structure) {
    is=title AnimGraphNode_RandomPlayer
    entries [
     {
      chanceToPlay=1
      maxLoopCount=0
      maxPlayRate=1
      minLoopCount=0
      minPlayRate=1
      sequence=/BRCosmetics/Accessories/FORT_Backpacks/Backpack_F_MED_GrowMallet/Animations/GrowMallet_IdleBreak1_Anim_PACK
      blendIn {blendOption=HermiteCubic, blendTime=0.2, customCurve=null}
     }
     {
      chanceToPlay=1
      maxLoopCount=0
      maxPlayRate=1
      minLoopCount=0
      minPlayRate=1
      sequence=/BRCosmetics/Accessories/FORT_Backpacks/Backpack_F_MED_GrowMallet/Animations/GrowMallet_IdleBreak2_Anim_PACK
      blendIn {blendOption=HermiteCubic, blendTime=0.2, customCurve=null}
     }
    ]
   }
   Output Animation Pose(Result=@n1)
  ]
 }
 Transition {path=AnimGraph.AnimGraphNode_StateMachine_0.FE Idle/Reaction.AnimStateTransitionNode_13.Transition, nodes [Result()]}
 AnimationTransitionGraph_1 {
  path=AnimGraph.AnimGraphNode_StateMachine_0.FE Idle/Reaction.AnimStateTransitionNode_3.AnimationTransitionGraph_1
  nodes [Result(bCanEnterTransition=OR Boolean(A=Get ChestOpened() -> (ChestOpened:bool), B=Get EnemyPlayerEliminated() -> (EnemyPlayerEliminated:bool)))]
 }
 AnimationTransitionGraph_0 {
  path=AnimGraph.AnimGraphNode_StateMachine_0.FE Idle/Reaction.AnimStateTransitionNode_6.AnimationTransitionGraph_0
  nodes [Result(bCanEnterTransition=Get IdleBreak() -> (IdleBreak:bool))]
 }
 Idle/reaction {
  entry=Idle
  path=AnimGraph.AnimGraphNode_StateMachine_1.Idle/reaction
  ids=id
  nodes [
   AnimStateNode_0 = Idle() {
    bAlwaysResetOnEntry=true
    stateEntered {guid=#1}
    stateFullyBlended {guid=#2}
    stateLeft {guid=#3}
   }
   AnimStateNode_1 = Elims/OpenChest() {is=title AnimStateNode}
   AnimStateNode_2 = Identity() {is=title AnimStateNode}
   AnimStateNode_3 = Idle Additives() {is=title AnimStateNode}
   AnimStateNode_5 = Idle Break() {is=title AnimStateNode, bAlwaysResetOnEntry=true}
   AnimStateNode_6 = Dammage Reactive() {is=title AnimStateNode}
   AnimStateTransitionNode_0 = AnimStateTransitionNode() {bAllowInertializationForSelfTransitions=true, logicType=TLT_Inertialization}
   AnimStateTransitionNode_1 = AnimStateTransitionNode() {bAllowInertializationForSelfTransitions=true, logicType=TLT_Inertialization}
   AnimStateTransitionNode_10 = AnimStateTransitionNode() {bAllowInertializationForSelfTransitions=true, bAutomaticRuleBasedOnSequencePlayerInState=true, logicType=TLT_Inertialization}
   AnimStateTransitionNode_12 = AnimStateTransitionNode() {bAllowInertializationForSelfTransitions=true, logicType=TLT_Inertialization}
   AnimStateTransitionNode_14 = AnimStateTransitionNode() {bAllowInertializationForSelfTransitions=true, logicType=TLT_Inertialization}
   AnimStateTransitionNode_15 = AnimStateTransitionNode() {bAllowInertializationForSelfTransitions=true, logicType=TLT_Inertialization}
   AnimStateTransitionNode_18 = AnimStateTransitionNode() {bAllowInertializationForSelfTransitions=true, logicType=TLT_Inertialization}
   AnimStateTransitionNode_2 = AnimStateTransitionNode() {bAllowInertializationForSelfTransitions=true, crossfadeDuration=0.5, logicType=TLT_Inertialization}
   AnimStateTransitionNode_3 = AnimStateTransitionNode() {bAllowInertializationForSelfTransitions=true, logicType=TLT_Inertialization}
   AnimStateTransitionNode_4 = AnimStateTransitionNode() {bAllowInertializationForSelfTransitions=true, logicType=TLT_Inertialization}
   AnimStateTransitionNode_6 = AnimStateTransitionNode() {bAllowInertializationForSelfTransitions=true, logicType=TLT_Inertialization}
   AnimStateTransitionNode_7 = AnimStateTransitionNode() {bAllowInertializationForSelfTransitions=true, logicType=TLT_Inertialization}
   AnimStateTransitionNode_8 = AnimStateTransitionNode() {bAllowInertializationForSelfTransitions=true, bAutomaticRuleBasedOnSequencePlayerInState=true, logicType=TLT_Inertialization}
   AnimStateTransitionNode_9 = AnimStateTransitionNode() {bAllowInertializationForSelfTransitions=true, bAutomaticRuleBasedOnSequencePlayerInState=true, logicType=TLT_Inertialization}
  ]
  links [
   {
    dst=Idle
    rule=AnimGraph.AnimGraphNode_StateMachine_1.Idle/reaction.AnimStateTransitionNode_13.Transition
    src=Dammage Reactive
    via=AnimStateTransitionNode_9
   }
   {
    dst=Idle
    rule=AnimGraph.AnimGraphNode_StateMachine_1.Idle/reaction.AnimStateTransitionNode_13.Transition
    src=Elims/OpenChest
    via=AnimStateTransitionNode_8
   }
   {
    dst=Idle
    rule=AnimGraph.AnimGraphNode_StateMachine_1.Idle/reaction.AnimStateTransitionNode_2.Transition
    src=Identity
    via=AnimStateTransitionNode_2
   }
   {
    dst=Idle Additives
    rule=AnimGraph.AnimGraphNode_StateMachine_1.Idle/reaction.AnimStateTransitionNode_0.Transition
    src=Identity
    via=AnimStateTransitionNode_0
   }
   {
    dst=Dammage Reactive
    rule=AnimGraph.AnimGraphNode_StateMachine_1.Idle/reaction.AnimStateTransitionNode_7.Transition
    src=Idle
    via=AnimStateTransitionNode_7
   }
   {
    dst=Elims/OpenChest
    rule=AnimGraph.AnimGraphNode_StateMachine_1.Idle/reaction.AnimStateTransitionNode_14.Transition
    src=Idle
    via=AnimStateTransitionNode_14
   }
   {
    dst=Identity
    rule=AnimGraph.AnimGraphNode_StateMachine_1.Idle/reaction.AnimStateTransitionNode_1.Transition
    src=Idle
    via=AnimStateTransitionNode_1
   }
   {
    dst=Idle Additives
    rule=AnimGraph.AnimGraphNode_StateMachine_1.Idle/reaction.AnimStateTransitionNode_4.Transition
    src=Idle
    via=AnimStateTransitionNode_4
   }
   {
    dst=Idle Break
    rule=AnimGraph.AnimGraphNode_StateMachine_1.Idle/reaction.AnimStateTransitionNode_6.AnimationTransitionGraph_0
    src=Idle
    via=AnimStateTransitionNode_6
   }
   {
    dst=Identity
    rule=AnimGraph.AnimGraphNode_StateMachine_1.Idle/reaction.AnimStateTransitionNode_3.Transition
    src=Idle Additives
    via=AnimStateTransitionNode_3
   }
   {
    dst=Dammage Reactive
    rule=AnimGraph.AnimGraphNode_StateMachine_1.Idle/reaction.AnimStateTransitionNode_7.Transition
    src=Idle Break
    via=AnimStateTransitionNode_12
   }
   {
    dst=Elims/OpenChest
    rule=AnimGraph.AnimGraphNode_StateMachine_1.Idle/reaction.AnimStateTransitionNode_14.Transition
    src=Idle Break
    via=AnimStateTransitionNode_15
   }
   {
    dst=Idle
    rule=AnimGraph.AnimGraphNode_StateMachine_1.Idle/reaction.AnimStateTransitionNode_13.Transition
    src=Idle Break
    via=AnimStateTransitionNode_10
   }
   {
    dst=Idle
    rule=AnimGraph.AnimGraphNode_StateMachine_1.Idle/reaction.AnimStateTransitionNode_18.Transition
    src=Idle Break
    via=AnimStateTransitionNode_18
   }
  ]
 }
 Idle {
  path=AnimGraph.AnimGraphNode_StateMachine_1.Idle/reaction.AnimStateNode_0.Idle
  nodes [Output Animation Pose(Result="Sequence Player (additive) 'GrowMallet_Idle_Anim_PACK'"() {sequence=&GrowMallet_Idle_Anim_PACK})]
 }
 Elims/OpenChest {
  path=AnimGraph.AnimGraphNode_StateMachine_1.Idle/reaction.AnimStateNode_1.Elims/OpenChest
  nodes [
   n1 = "Sequence Player (additive) 'GrowMallet_Elimination_Anim_PACK'"() {bLoopAnimation=false, sequence=/BRCosmetics/Accessories/FORT_Backpacks/Backpack_F_MED_GrowMallet/Animations/GrowMallet_Elimination_Anim_PACK}
   Output Animation Pose(Result=@n1)
  ]
 }
 Identity {
  path=AnimGraph.AnimGraphNode_StateMachine_1.Idle/reaction.AnimStateNode_2.Identity
  nodes [Output Animation Pose(Result=Additive Identity Pose())]
 }
 Idle Additives {
  path=AnimGraph.AnimGraphNode_StateMachine_1.Idle/reaction.AnimStateNode_3.Idle Additives
  nodes [Output Animation Pose(Result="Sequence Player (additive) 'GrowMallet_IdleForAdditives_Anim_PACK'"() {sequence=/BRCosmetics/Accessories/FORT_Backpacks/Backpack_F_MED_GrowMallet/Animations/GrowMallet_IdleForAdditives_Anim_PACK})]
 }
 Idle Break {
  path=AnimGraph.AnimGraphNode_StateMachine_1.Idle/reaction.AnimStateNode_5.Idle Break
  nodes [
   n1 = Random Sequence Player() {entries [
     {
      chanceToPlay=1
      maxLoopCount=0
      maxPlayRate=1
      minLoopCount=0
      minPlayRate=1
      sequence=&GrowMallet_IdleBreak1_Anim_PACK
      blendIn {blendOption=HermiteCubic, blendTime=0.2, customCurve=null}
     }
     {
      chanceToPlay=1
      maxLoopCount=0
      maxPlayRate=1
      minLoopCount=0
      minPlayRate=1
      sequence=&GrowMallet_IdleBreak2_Anim_PACK
      blendIn {blendOption=HermiteCubic, blendTime=0.2, customCurve=null}
     }
    ]}
   Output Animation Pose(Result=@n1)
  ]
 }
 Dammage Reactive {
  path=AnimGraph.AnimGraphNode_StateMachine_1.Idle/reaction.AnimStateNode_6.Dammage Reactive
  nodes [
   n1 = "Sequence Player (additive) 'GrowMallet_Damage_Anim_PACK'"() {bLoopAnimation=false, sequence=/BRCosmetics/Accessories/FORT_Backpacks/Backpack_F_MED_GrowMallet/Animations/GrowMallet_Damage_Anim_PACK}
   Output Animation Pose(Result=@n1)
  ]
 }
 Transition {
  path=AnimGraph.AnimGraphNode_StateMachine_1.Idle/reaction.AnimStateTransitionNode_0.Transition
  nodes [Result(bCanEnterTransition=Get IdleAdditives() -> (IdleAdditives:bool))]
 }
 Transition {
  path=AnimGraph.AnimGraphNode_StateMachine_1.Idle/reaction.AnimStateTransitionNode_1.Transition
  nodes [Result(bCanEnterTransition=Get IdleStop() -> (IdleStop:bool))]
 }
 Transition {path=AnimGraph.AnimGraphNode_StateMachine_1.Idle/reaction.AnimStateTransitionNode_13.Transition, nodes [Result()]}
 Transition {path=AnimGraph.AnimGraphNode_StateMachine_1.Idle/reaction.AnimStateTransitionNode_13.Transition, nodes [Result()]}
 Transition {
  path=AnimGraph.AnimGraphNode_StateMachine_1.Idle/reaction.AnimStateTransitionNode_14.Transition
  nodes [Result(bCanEnterTransition=OR Boolean(A=Get ChestOpened(), B=Get EnemyPlayerEliminated()))]
 }
 Transition {
  path=AnimGraph.AnimGraphNode_StateMachine_1.Idle/reaction.AnimStateTransitionNode_18.Transition
  nodes [Result(bCanEnterTransition=OR Boolean(A=Get PlayerDBNO() -> (PlayerDBNO:bool), B="Get Player  Skydiving"() -> ("Player  Skydiving":bool)))]
 }
 Transition {
  path=AnimGraph.AnimGraphNode_StateMachine_1.Idle/reaction.AnimStateTransitionNode_2.Transition
  nodes [
   n1 = Get IdleAdditives()
   n2 = Get IdleStop()
   Result(bCanEnterTransition=AND Boolean(A=NOT Boolean(A=@n2), B=NOT Boolean(A=@n1)))
  ]
 }
 Transition {
  path=AnimGraph.AnimGraphNode_StateMachine_1.Idle/reaction.AnimStateTransitionNode_3.Transition
  nodes [
   n1 = Get IdleAdditives()
   Result(bCanEnterTransition=AND Boolean(A=Get IdleStop(), B=NOT Boolean(A=@n1)))
  ]
 }
 Transition {
  path=AnimGraph.AnimGraphNode_StateMachine_1.Idle/reaction.AnimStateTransitionNode_4.Transition
  nodes [Result(bCanEnterTransition=Get IdleAdditives())]
 }
 AnimationTransitionGraph_0 {
  path=AnimGraph.AnimGraphNode_StateMachine_1.Idle/reaction.AnimStateTransitionNode_6.AnimationTransitionGraph_0
  nodes [
   n1 = OR Boolean(A="Get Player  Skydiving"(), B=Get PlayerDBNO())
   Result(bCanEnterTransition=AND Boolean!(A=NOT Boolean(A=@n1), B=Get IdleBreak(), C:bool=Get Can Idle Break() -> (Can Idle Break:bool)) -> (ReturnValue))
  ]
 }
 Transition {
  path=AnimGraph.AnimGraphNode_StateMachine_1.Idle/reaction.AnimStateTransitionNode_7.Transition
  nodes [
   n1 = Get PlayerDBNO()
   n2 = Get ShieldsDecreased() -> (ShieldsDecreased:bool)
   n3 = Get HealthDecreased() -> (HealthDecreased:bool)
   Result(bCanEnterTransition=AND Boolean!(A=NOT Boolean(A=@n1), B=OR Boolean(A=@n3, B=@n2)) -> (ReturnValue))
  ]
 }
 Reactive State Machine {
  entry=Identity
  path=AnimGraph.AnimGraphNode_StateMachine_2.Reactive State Machine
  ids=id
  nodes [
   To Identity() {is=title AnimStateAliasNode, also {aliases [Crouch Out, DBNO Out, Glide Out, Jump]}}
   To DBNO() {is=title AnimStateAliasNode, also {aliases [DBNO Out, Identity]}}
   To Crouch() {is=title AnimStateAliasNode, also {aliases [Crouch Out, Identity]}}
   To Glide() {is=title AnimStateAliasNode, also {aliases [Glide Out, Identity, Jump]}}
   To Jump() {is=title AnimStateAliasNode, also {aliases [Identity,Jump]}}
   AnimStateNode_0 = Identity()
   AnimStateNode_1 = Glide Into() {is=title AnimStateNode}
   AnimStateNode_10 = Jump() {is=title AnimStateNode, bAlwaysResetOnEntry=true}
   AnimStateNode_2 = DBNO Into() {is=title AnimStateNode}
   AnimStateNode_3 = DBNO() {is=title AnimStateNode}
   AnimStateNode_4 = DBNO Out() {is=title AnimStateNode}
   AnimStateNode_5 = Crouch Into() {is=title AnimStateNode}
   AnimStateNode_6 = Crouch Out() {is=title AnimStateNode}
   AnimStateNode_8 = Glide() {is=title AnimStateNode}
   AnimStateNode_9 = Glide Out() {is=title AnimStateNode}
   AnimStateTransitionNode_0 = AnimStateTransitionNode() {bAllowInertializationForSelfTransitions=true, logicType=TLT_Inertialization}
   AnimStateTransitionNode_1 = AnimStateTransitionNode() {bAllowInertializationForSelfTransitions=true, logicType=TLT_Inertialization}
   AnimStateTransitionNode_10 = AnimStateTransitionNode() {bAllowInertializationForSelfTransitions=true, logicType=TLT_Inertialization}
   AnimStateTransitionNode_12 = AnimStateTransitionNode() {bAllowInertializationForSelfTransitions=true, logicType=TLT_Inertialization}
   AnimStateTransitionNode_2 = AnimStateTransitionNode() {bAllowInertializationForSelfTransitions=true, logicType=TLT_Inertialization}
   AnimStateTransitionNode_3 = AnimStateTransitionNode() {bAllowInertializationForSelfTransitions=true, bAutomaticRuleBasedOnSequencePlayerInState=true, logicType=TLT_Inertialization}
   AnimStateTransitionNode_4 = AnimStateTransitionNode() {bAllowInertializationForSelfTransitions=true, bAutomaticRuleBasedOnSequencePlayerInState=true, logicType=TLT_Inertialization}
   AnimStateTransitionNode_5 = AnimStateTransitionNode() {bAllowInertializationForSelfTransitions=true, logicType=TLT_Inertialization}
   AnimStateTransitionNode_7 = AnimStateTransitionNode() {bAllowInertializationForSelfTransitions=true, bAutomaticRuleBasedOnSequencePlayerInState=true, logicType=TLT_Inertialization}
   AnimStateTransitionNode_8 = AnimStateTransitionNode() {bAllowInertializationForSelfTransitions=true, logicType=TLT_Inertialization}
   AnimStateTransitionNode_9 = AnimStateTransitionNode() {bAllowInertializationForSelfTransitions=true, logicType=TLT_Inertialization}
  ]
  links [
   {
    dst=Crouch Out
    rule=AnimGraph.AnimGraphNode_StateMachine_2.Reactive State Machine.AnimStateTransitionNode_1.AnimationTransitionGraph_0
    src=Crouch Into
    via=AnimStateTransitionNode_1
   }
   {
    dst=DBNO Out
    rule=AnimGraph.AnimGraphNode_StateMachine_2.Reactive State Machine.AnimStateTransitionNode_5.AnimationTransitionGraph_0
    src=DBNO
    via=AnimStateTransitionNode_5
   }
   {
    dst=DBNO
    rule=AnimGraph.AnimGraphNode_StateMachine_1.Idle/reaction.AnimStateTransitionNode_13.Transition
    src=DBNO Into
    via=AnimStateTransitionNode_4
   }
   {
    dst=DBNO Out
    rule=AnimGraph.AnimGraphNode_StateMachine_2.Reactive State Machine.AnimStateTransitionNode_2.Transition
    src=DBNO Into
    via=AnimStateTransitionNode_2
   }
   {
    dst=Glide Out
    rule=AnimGraph.AnimGraphNode_StateMachine_2.Reactive State Machine.AnimStateTransitionNode_9.AnimationTransitionGraph_0
    src=Glide
    via=AnimStateTransitionNode_9
   }
   {
    dst=Glide
    rule=AnimGraph.AnimGraphNode_StateMachine_1.Idle/reaction.AnimStateTransitionNode_13.Transition
    src=Glide Into
    via=AnimStateTransitionNode_3
   }
   {
    dst=Crouch Into
    rule=AnimGraph.AnimGraphNode_StateMachine_2.Reactive State Machine.AnimStateTransitionNode_0.AnimationTransitionGraph_0
    src=To Crouch
    via=AnimStateTransitionNode_0
   }
   {
    dst=DBNO Into
    rule=AnimGraph.AnimGraphNode_StateMachine_2.Reactive State Machine.AnimStateTransitionNode_8.AnimationTransitionGraph_0
    src=To DBNO
    via=AnimStateTransitionNode_8
   }
   {
    dst=Glide Into
    rule=AnimGraph.AnimGraphNode_StateMachine_2.Reactive State Machine.AnimStateTransitionNode_10.AnimationTransitionGraph_0
    src=To Glide
    via=AnimStateTransitionNode_10
   }
   {
    dst=Identity
    rule=AnimGraph.AnimGraphNode_StateMachine_1.Idle/reaction.AnimStateTransitionNode_13.Transition
    src=To Identity
    via=AnimStateTransitionNode_7
   }
   {
    dst=Jump
    rule=AnimGraph.AnimGraphNode_StateMachine_2.Reactive State Machine.AnimStateTransitionNode_12.AnimationTransitionGraph_0
    src=To Jump
    via=AnimStateTransitionNode_12
   }
  ]
 }
 Identity {
  path=AnimGraph.AnimGraphNode_StateMachine_2.Reactive State Machine.AnimStateNode_0.Identity
  nodes [Output Animation Pose(Result=Additive Identity Pose())]
 }
 Glide Into {
  path=AnimGraph.AnimGraphNode_StateMachine_2.Reactive State Machine.AnimStateNode_1.Glide Into
  nodes [
   n1 = "Sequence Player (additive) 'GrowMallet_GlideInto_Anim_PACK'"() {bLoopAnimation=false, sequence=/BRCosmetics/Accessories/FORT_Backpacks/Backpack_F_MED_GrowMallet/Animations/GrowMallet_GlideInto_Anim_PACK}
   Output Animation Pose(Result=@n1)
  ]
 }
 Jump {
  path=AnimGraph.AnimGraphNode_StateMachine_2.Reactive State Machine.AnimStateNode_10.Jump
  nodes [
   n1 = "Sequence Player (additive) 'GrowMallet_Jump_Anim_PACK'"() {
    bLoopAnimation=false
    sequence=/BRCosmetics/Accessories/FORT_Backpacks/Backpack_F_MED_GrowMallet/Animations/GrowMallet_Jump_Anim_PACK
    startPosition=0.18
   }
   Output Animation Pose(Result=@n1)
  ]
 }
 DBNO Into {
  path=AnimGraph.AnimGraphNode_StateMachine_2.Reactive State Machine.AnimStateNode_2.DBNO Into
  nodes [
   n1 = "Sequence Player (additive) 'GrowMallet_DBNOInto_Anim_PACK'"() {bLoopAnimation=false, sequence=/BRCosmetics/Accessories/FORT_Backpacks/Backpack_F_MED_GrowMallet/Animations/GrowMallet_DBNOInto_Anim_PACK}
   Output Animation Pose(Result=@n1)
  ]
 }
 DBNO {
  path=AnimGraph.AnimGraphNode_StateMachine_2.Reactive State Machine.AnimStateNode_3.DBNO
  nodes [Output Animation Pose(Result="Sequence Player (additive) 'GrowMallet_DBNO_Anim_PACK'"() {sequence=/BRCosmetics/Accessories/FORT_Backpacks/Backpack_F_MED_GrowMallet/Animations/GrowMallet_DBNO_Anim_PACK})]
 }
 DBNO Out {
  path=AnimGraph.AnimGraphNode_StateMachine_2.Reactive State Machine.AnimStateNode_4.DBNO Out
  nodes [
   n1 = "Sequence Player (additive) 'GrowMallet_DBNOOut_Anim_PACK'"() {bLoopAnimation=false, sequence=/BRCosmetics/Accessories/FORT_Backpacks/Backpack_F_MED_GrowMallet/Animations/GrowMallet_DBNOOut_Anim_PACK}
   Output Animation Pose(Result=@n1)
  ]
 }
 Crouch Into {
  path=AnimGraph.AnimGraphNode_StateMachine_2.Reactive State Machine.AnimStateNode_5.Crouch Into
  nodes [
   n1 = "Sequence Player (additive) 'GrowMallet_CrouchInto_Anim_PACK'"() {bLoopAnimation=false, sequence=/BRCosmetics/Accessories/FORT_Backpacks/Backpack_F_MED_GrowMallet/Animations/GrowMallet_CrouchInto_Anim_PACK}
   Output Animation Pose(Result=@n1)
  ]
 }
 Crouch Out {
  path=AnimGraph.AnimGraphNode_StateMachine_2.Reactive State Machine.AnimStateNode_6.Crouch Out
  nodes [
   n1 = "Sequence Player (additive) 'GrowMallet_CrouchOut_Anim_PACK'"() {bLoopAnimation=false, sequence=/BRCosmetics/Accessories/FORT_Backpacks/Backpack_F_MED_GrowMallet/Animations/GrowMallet_CrouchOut_Anim_PACK}
   Output Animation Pose(Result=@n1)
  ]
 }
 Glide {
  path=AnimGraph.AnimGraphNode_StateMachine_2.Reactive State Machine.AnimStateNode_8.Glide
  nodes [Output Animation Pose(Result="Sequence Player (additive) 'GrowMallet_Glide_Anim_PACK'"() {sequence=/BRCosmetics/Accessories/FORT_Backpacks/Backpack_F_MED_GrowMallet/Animations/GrowMallet_Glide_Anim_PACK})]
 }
 Glide Out {
  path=AnimGraph.AnimGraphNode_StateMachine_2.Reactive State Machine.AnimStateNode_9.Glide Out
  nodes [
   n1 = "Sequence Player (additive) 'GrowMallet_GlideOut_Anim_PACK'"() {bLoopAnimation=false, sequence=/BRCosmetics/Accessories/FORT_Backpacks/Backpack_F_MED_GrowMallet/Animations/GrowMallet_GlideOut_Anim_PACK}
   Output Animation Pose(Result=@n1)
  ]
 }
 AnimationTransitionGraph_0 {
  path=AnimGraph.AnimGraphNode_StateMachine_2.Reactive State Machine.AnimStateTransitionNode_0.AnimationTransitionGraph_0
  nodes [
   n1 = Get bIsMoving2D()
   n2 = Get PlayerCrouch() -> (PlayerCrouch:bool)
   n3 = NOT Boolean(A=Get PlayerDBNO())
   n4 = Get PlayerCrouch()
   Result(bCanEnterTransition=OR Boolean(A=AND Boolean(A=@n2, B=@n1), B=AND Boolean(A=@n3, B=@n4)))
  ]
 }
 AnimationTransitionGraph_0 {
  path=AnimGraph.AnimGraphNode_StateMachine_2.Reactive State Machine.AnimStateTransitionNode_1.AnimationTransitionGraph_0
  nodes [
   n1 = Get PlayerCrouch()
   n2 = NOT Boolean(A=@n1)
   Result(bCanEnterTransition=@n2)
   AND Boolean(A=@n1, B=@n2)
   Get bIsMoving2D()
  ]
 }
 AnimationTransitionGraph_0 {
  path=AnimGraph.AnimGraphNode_StateMachine_2.Reactive State Machine.AnimStateTransitionNode_10.AnimationTransitionGraph_0
  nodes [Result(bCanEnterTransition="Get Player  Skydiving"())]
 }
 AnimationTransitionGraph_0 {
  path=AnimGraph.AnimGraphNode_StateMachine_2.Reactive State Machine.AnimStateTransitionNode_12.AnimationTransitionGraph_0
  nodes [Result(bCanEnterTransition=Get Player Jump() -> (Player Jump:bool))]
 }
 Transition {
  path=AnimGraph.AnimGraphNode_StateMachine_2.Reactive State Machine.AnimStateTransitionNode_2.Transition
  nodes [Result(bCanEnterTransition=NOT Boolean(A=Get PlayerDBNO()))]
 }
 AnimationTransitionGraph_0 {
  path=AnimGraph.AnimGraphNode_StateMachine_2.Reactive State Machine.AnimStateTransitionNode_5.AnimationTransitionGraph_0
  nodes [Result(bCanEnterTransition=NOT Boolean(A=Get PlayerDBNO()))]
 }
 AnimationTransitionGraph_0 {
  path=AnimGraph.AnimGraphNode_StateMachine_2.Reactive State Machine.AnimStateTransitionNode_8.AnimationTransitionGraph_0
  nodes [Result(bCanEnterTransition=Get PlayerDBNO())]
 }
 AnimationTransitionGraph_0 {
  path=AnimGraph.AnimGraphNode_StateMachine_2.Reactive State Machine.AnimStateTransitionNode_9.AnimationTransitionGraph_0
  nodes [Result(bCanEnterTransition=NOT Boolean(A="Get Player  Skydiving"()))]
 }
 EventGraph {nodes [
   Event BlueprintUpdateAnimation()
   > n1 = Sequence!(execute) -> (then_0, then_1, then_2:Exec, then_3:Exec, then_4:Exec, then_5:Exec, then_6:Exec, then_7:Exec)
   n2 = GetOwnerSkeletalMeshComponent()
   n3 = BreakRotator(InRot=BreakTransform(InTransform=GetDeltaTransformFromRefPose(self=@n2, BoneName=spine_05, BaseName=root)).Rotation)
   Set Gimbal Rot(execute=@n1.then_1, Gimbal Rot:Rotator=MakeRotator(Pitch=float * float(A=@n3.Yaw, B=-0.8) {is=title K2Node_PromotableOperator})) -> (Output_Get:Rotator)
   > Branch(Condition=NOT Boolean(A=Get bIsCrouching()))
   > Set ADSGimbal(ADSGimbal:Rotator=MakeRotator(Pitch=float * float(A=@n3.Yaw, B=0.3))) -> (Output_Get:Rotator)
   n4 = "Gravity Override State Function (Parameter Asset) (Nativized)"(
    execute=@n1.then_0
    CharacterPartAnimInstance=Self-Reference()
    GameTimeInSeconds=GetGameTimeInSeconds()
    Parameters=&/F_MED_GrowMallet_Pack_Parameters
   )
   > Set Gravity Override(Gravity Override=@n4)
   n5 = Branch(execute=@n1.then_4, Condition=Get bShouldDisplayFrontendReactivity())
   > Do Once(Reset=@n5.else)
   > Delay(Duration=4)
   > Set EnemyPlayerEliminated(EnemyPlayerEliminated:bool=true) -> (Output_Get:bool)
   Branch(execute=@n1.then_5, Condition=Get Can Idle Break())
   > SetIdleBreak()
   > Set Physics(Physics:float=FInterpTo(
     Current=Get Current_0() -> (Current_0:double)
     Target=GetCurveValue(CurveName=PlayIdle)
     DeltaTime=GetWorldDeltaSeconds!() -> (ReturnValue:double)
     InterpSpeed=40
    )) -> (Output_Get:float)
   n6 = GetCurveValue(CurveName=lowerblinkl)
   n7 = GetCurveValue(CurveName=lowerblinkr)
   n8 = OR Boolean(
    A="Not Equal (Float)"(A=@n7, B=Get PrevLowerRBlinkPose() -> (PrevLowerRBlinkPose:float))
    B="Not Equal (Float)"(A=@n6, B=Get PrevLowerLBlinkPose() -> (PrevLowerLBlinkPose:float))
   )
   n9 = GetCurveValue(CurveName=upperblinkr)
   n10 = Get PrevUpperRBlinkPose() -> (PrevUpperRBlinkPose:float)
   n11 = GetCurveValue(CurveName=upperblinkl)
   n12 = Get PrevUpperLBlinkPose() -> (PrevUpperLBlinkPose:float)
   n13 = MakeVector2D(
    X=float - float(A=GetCurveValue(CurveName=lookleft), B=GetCurveValue(CurveName=lookright))
    Y=float - float(A=GetCurveValue(CurveName=lookup), B=GetCurveValue(CurveName=lookdown))
   )
   n14 = OR Boolean(
    A="Not Equal (Vector2D)"!(
     A:Vector 2D Structure=@n13
     B:Vector 2D Structure=Get PrevLookDirection() -> (PrevLookDirection:Vector 2D Structure)
     ErrorTolerance:float=0.025
    ) -> (ReturnValue:bool=~) {is=title K2Node_PromotableOperator}
    B=OR Boolean(A="Not Equal (Float)"(A=@n9, B=@n10), B="Not Equal (Float)"(A=@n11, B=@n12))
   )
   SetScalarParameterValueOnMaterials!(
    execute=@n1.then_6
    self:Mesh Component Object Reference=GetOwningComponent()
    ParameterName:Name=HasLookDirectionChanged
    ParameterValue:float="To Float (Boolean)"(InBool=OR Boolean(A=@n14, B=@n8))
   ) -> (then)
   > Set PrevLookDirection(PrevLookDirection:Vector 2D Structure=@n13) -> (Output_Get:Vector 2D Structure)
   > Set PrevUpperRBlinkPose(PrevUpperRBlinkPose:float=@n9) -> (Output_Get:float)
   > Set PrevLowerRBlinkPose(PrevLowerRBlinkPose:float=@n7) -> (Output_Get:float)
   > Set PrevUpperLBlinkPose(PrevUpperLBlinkPose:float=@n11) -> (Output_Get:float)
   > Set PrevLowerLBlinkPose(PrevLowerLBlinkPose:float=@n6) -> (Output_Get:float)
   Branch(execute=@n1.then_2, Condition=NOT Boolean(A=Get bIsSkyDiving()))
   > Set Rot Down(Rot Down:float=GetAlphaFromReferenceTransformDelta(
     SkeletalMeshComponent="Get Skel mesh "() -> ("Skel mesh ":Skeletal Mesh Component Object Reference)
     BoneName=spine_03
     BaseName="head "
     InputRange=(X=0,Y=-120)
     OutputRange=(X=0,Y=-1)
     TwistAxis="0,0,1"
    )) -> (Output_Get:float)
   > Set Rot Up(Rot Up:float=GetAlphaFromReferenceTransformDelta(SkeletalMeshComponent="Get Skel mesh "(), BoneName=spine_04, BaseName=head, InputRange=(X=15,Y=60), TwistAxis="0,0,1")) -> (Output_Get:float)
   GetPlayerMotion(execute=@n1.then_3)
   > GetTailsBoneTransform()
   n15 = Event BlueprintBeginPlay() {is=title K2Node_Event}
   Set IdleBreakLimitation(
    execute=[@n15.then, @n18.then]
    IdleBreakLimitation:float=RandomFloatInRange(Min=7, Max=Get IdleBreakLimitationMax() -> (IdleBreakLimitationMax:float))
   ) -> (Output_Get:float)
   > Set IdleBreakDuration(IdleBreakDuration:float=0) -> (Output_Get:float)
   AnimNotify_IdleAdditives() {is=title K2Node_Event}
   > Set IdleAdditives(IdleAdditives:bool=true) -> (Output_Get:bool)
   AnimNotify_FE_Elim_End() {is=title K2Node_Event}
   > Branch(Condition=Get bShouldDisplayFrontendReactivity())
   > Delay(Duration=10)
   > Set EnemyPlayerEliminated(EnemyPlayerEliminated=true)
   Event BlueprintInitializeAnimation()
   > n16 = Cast To FortPlayerPawn(Object=TryGetFortPawn())
   > Set As Fort Player Pawn(As Fort Player Pawn:Fort Player Pawn Object Reference=@n16) -> (Output_Get:Fort Player Pawn Object Reference)
   > "Set Skel mesh "("Skel mesh ":Skeletal Mesh Component Object Reference=GetSkeletalMeshForPartType(self=@n16)) -> (Output_Get:Skeletal Mesh Component Object Reference)
   > Set IdleBreakDuration(IdleBreakDuration=0)
   n17 = Event BlueprintProcessAnimInputEvent()
   > GetInputEvent!(execute, self, GetInputEvent:Fort Anim Input Event Object Reference=@n17.AnimInputEvent) -> (then)
   n18 = AnimNotify_IdleBreakEnd() {is=title K2Node_Event}
   Event BlueprintPostEvaluateAnimation() {is=title K2Node_Event}
   > SetVariablesToFalse()
   Event OnBeginDBNO() {is=title K2Node_Event}
   > Set PlayerDBNO(PlayerDBNO:bool=true) -> (Output_Get:bool)
   Event OnEndDBNO() {is=title K2Node_Event}
   > Set PlayerDBNO(PlayerDBNO=false)
   AnimNotify_IdleStop() {is=title K2Node_Event}
   > Set IdleStop(IdleStop:bool=true) -> (Output_Get:bool)
   AnimNotify_IdleStart() {is=title K2Node_Event}
   > Set IdleStop(IdleStop=false)
   AnimNotify_EndTransition() {is=title K2Node_Event}
   > Set Can Idle Break(Can Idle Break:bool=false) -> (Output_Get:bool)
   > Delay(Duration=4)
   > Set Can Idle Break(Can Idle Break=true)
  ]}
 GetInputEvent {nodes [
   n1 = GetInputEvent!() -> (then, GetInputEvent:Fort Anim Input Event Object Reference) {is=title K2Node_FunctionEntry}
   > Set InputEvent(InputEvent:Fort Anim Input Event Object Reference=@n1) -> (Output_Get:Fort Anim Input Event Object Reference)
   > n2 = Sequence!(execute) -> (then_0, then_1, then_2, then_3, then_4, then_5)
   n3 = Get InputEvent() -> (InputEvent:Fort Anim Input Event Object Reference)
   > n4 = Branch(Condition=IsNamed(self=@n3, InEventName=aie.ChestOpened))
   > Set ChestOpened(ChestOpened:bool=true) -> (Output_Get:bool)
   Set ChestOpened(execute=@n4.else, ChestOpened=false)
   n5 = Branch(execute=@n2.then_1, Condition=IsNamed(self=@n3, InEventName=aie.EnemyPlayerEliminated))
   > Set EnemyPlayerEliminated(EnemyPlayerEliminated=true)
   Set EnemyPlayerEliminated(execute=@n5.else, EnemyPlayerEliminated=false)
   n6 = Branch(execute=@n2.then_2, Condition=IsNamed(self=@n3, InEventName=aie.PlayerVitality.ShieldsIncreased))
   > Set ShieldsIncreased(ShieldsIncreased:bool=true) -> (Output_Get:bool)
   Set ShieldsIncreased(execute=@n6.else, ShieldsIncreased=false)
   n7 = Branch(execute=@n2.then_3, Condition=IsNamed(self=@n3, InEventName=aie.PlayerVitality.ShieldsDecreased))
   > Set ShieldsDecreased(ShieldsDecreased:bool=true) -> (Output_Get:bool)
   Set ShieldsDecreased(execute=@n7.else, ShieldsDecreased=false)
   n8 = Branch(execute=@n2.then_4, Condition=IsNamed(self=@n3, InEventName=aie.PlayerVitality.HealthIncreased))
   > Set HealthIncreased(HealthIncreased:bool=true) -> (Output_Get:bool)
   Set HealthIncreased(execute=@n8.else, HealthIncreased=false)
   n9 = Branch(execute=@n2.then_5, Condition=IsNamed(self=@n3, InEventName=aie.PlayerVitality.HealthDecreased))
   > Set HealthDecreased(HealthDecreased:bool=true) -> (Output_Get:bool)
   Set HealthDecreased(execute=@n9.else, HealthDecreased=false)
  ]}
 GetIsReactive {nodes [
   GetIsReactive() {is=title K2Node_FunctionEntry}
   > Return Node!(
    execute
    IsReactive:bool=OR Boolean!(
     A=Get ChestOpened()
     B=Get HealthDecreased()
     C:bool=Get EnemyPlayerEliminated()
     D:bool=Get ShieldsDecreased()
     E:bool=Get PlayerDBNO()
     F:bool=Get Player Jump()
     G:bool=Get PlayerCrouch()
     H:bool="Get Player  Skydiving"()
    ) -> (ReturnValue)
   )
  ]}
 GetPlayerMotion {nodes [
   GetPlayerMotion() {is=title K2Node_FunctionEntry}
   > n1 = Sequence!(execute) -> (then_0, then_1, then_2, then_3, then_4)
   > n2 = Branch(Condition=Get bIsMoving2D())
   > Set Is Moving 2D(Is Moving 2D:bool=true) -> (Output_Get:bool)
   Set Is Moving 2D(execute=@n2.else, Is Moving 2D=false)
   n3 = Branch(
    execute=@n1.then_4
    Condition=Get bPressedJump!(self:Character Object Reference=Get As Fort Player Pawn() -> (As Fort Player Pawn:Fort Player Pawn Object Reference)) -> (bPressedJump:bool)
   )
   > Set Player Jump(Player Jump:bool=true) -> (Output_Get:bool)
   Set Player Jump(execute=@n3.else, Player Jump=false)
   n4 = Branch(execute=@n1.then_3, Condition=IsSprinting!(self:Fort Pawn Object Reference=Get As Fort Player Pawn()) -> (ReturnValue:bool))
   > Set PlayerSprint(PlayerSprint:bool=true) -> (Output_Get:bool)
   Set PlayerSprint(execute=@n4.else, PlayerSprint=false)
   n5 = Branch(execute=@n1.then_2, Condition=Get bIsSkydiving!(self:Fort Player Pawn Object Reference=Get As Fort Player Pawn()) -> (bIsSkydiving:bool))
   > "Set Player  Skydiving"("Player  Skydiving":bool=true) -> (Output_Get:bool)
   "Set Player  Skydiving"(execute=@n5.else, "Player  Skydiving"=false)
   n6 = Branch(execute=@n1.then_1, Condition=Get bIsCrouched!(self:Character Object Reference=Get As Fort Player Pawn()) -> (bIsCrouched:bool))
   > Set PlayerCrouch(PlayerCrouch:bool=true) -> (Output_Get:bool)
   Set PlayerCrouch(execute=@n6.else, PlayerCrouch=false)
  ]}
 GetTailsBoneTransform {nodes [
   GetTailsBoneTransform() {is=title K2Node_FunctionEntry}
   > Set BaseTail(BaseTail:Transform=GetBoneTransform!(
     self:Skinned Mesh Component Object Reference="Get Skel mesh "()
     InBoneName:Name=base_tail
     TransformSpace:ERelativeTransformSpace Enum=RTS_Component
    ) -> (ReturnValue:Transform)) -> (Output_Get:Transform)
   > Set dyn_tail_01(dyn_tail_01:Transform=GetBoneTransform(self="Get Skel mesh "(), InBoneName=dyn_tail_01, TransformSpace=RTS_Component)) -> (Output_Get:Transform)
   > Set dyn_tail_02(dyn_tail_02:Transform=GetBoneTransform(self="Get Skel mesh "(), InBoneName=dyn_tail_02, TransformSpace=RTS_Component)) -> (Output_Get:Transform)
   > Set dyn_tail_03(dyn_tail_03:Transform=GetBoneTransform(self="Get Skel mesh "(), InBoneName=dyn_tail_03, TransformSpace=RTS_Component)) -> (Output_Get:Transform)
   > Set dyn_tail_04(dyn_tail_04:Transform=GetBoneTransform(self="Get Skel mesh "(), InBoneName=dyn_tail_04, TransformSpace=RTS_Component)) -> (Output_Get:Transform)
   > Set dyn_tailfin_l_01(dyn_tailfin_l_01:Transform=GetBoneTransform(self="Get Skel mesh "(), InBoneName=dyn_tailfin_l_01, TransformSpace=RTS_Component)) -> (Output_Get:Transform)
   > Set dyn_tailfin_l_02(dyn_tailfin_l_02:Transform=GetBoneTransform(self="Get Skel mesh "(), InBoneName=dyn_tailfin_l_02, TransformSpace=RTS_Component)) -> (Output_Get:Transform)
   > Set dyn_tailfin_r_01(dyn_tailfin_r_01:Transform=GetBoneTransform(self="Get Skel mesh "(), InBoneName=dyn_tailfin_r_01, TransformSpace=RTS_Component)) -> (Output_Get:Transform)
   > Set dyn_tailfin_r_02(dyn_tailfin_r_02:Transform=GetBoneTransform(self="Get Skel mesh "(), InBoneName=dyn_tailfin_r_02, TransformSpace=RTS_Component)) -> (Output_Get:Transform)
  ]}
 SetIdleBreak {nodes [
   SetIdleBreak() {is=title K2Node_FunctionEntry}
   n1 = Get IdleBreak()
   > Branch(Condition=AND Boolean(A=NOT Boolean(A=@n1), B=Get Can Idle Break()))
   > n2 = Branch(Condition="float > float"(A=Get IdleBreakDuration() -> (IdleBreakDuration:float), B=Get IdleBreakLimitationMax()))
   > n3 = Set IdleBreakDuration(IdleBreakDuration=0)
   n4 = Set IdleBreakDuration(execute=[@n3.then, @n2.else], IdleBreakDuration=float + float(A=Get IdleBreakDuration(), B=GetWorldDeltaSeconds()))
   > n5 = Branch(Condition="float > float"(A=@n4, B=Get IdleBreakLimitation() -> (IdleBreakLimitation:float)))
   n6 = Get IdleStop()
   > n7 = Branch(Condition=AND Boolean(A=NOT Boolean(A=@n6), B=Get Can Idle Break()))
   > Set IdleBreak(IdleBreak:bool=true) -> (Output_Get:bool)
   > Set IdleBreakDuration(IdleBreakDuration=0)
   Set IdleBreak(execute=@n7.else, IdleBreak=false)
   Branch(execute=@n5.else, Condition="float >= float"(A=Get IdleBreakDuration(), B=Get IdleBreakLimitationMax()))
   > Set IdleBreakDuration(IdleBreakDuration=0)
  ]}
 SetVariablesToFalse {nodes [
   SetVariablesToFalse() {is=title K2Node_FunctionEntry}
   > n1 = Sequence!(execute) -> (then_0, then_1, then_2)
   > Set ChestOpened(ChestOpened=false)
   > Set EnemyPlayerEliminated(EnemyPlayerEliminated=false)
   > Set ShieldsIncreased(ShieldsIncreased=false)
   > Set ShieldsDecreased(ShieldsDecreased=false)
   > Set HealthIncreased(HealthIncreased=false)
   > Set HealthDecreased(HealthDecreased=false)
   Set Is Moving 2D(execute=@n1.then_1, Is Moving 2D=false)
   > Set PlayerCrouch(PlayerCrouch=false)
   > Set PlayerSprint(PlayerSprint=false)
   > Set IdleBreak(IdleBreak=false)
   > Set IdleAdditives(IdleAdditives=false)
  ]}
]
variables [
 ADSGimbal {type=Rotator}
 As Fort Player Pawn {type=FortPlayerPawn}
 BaseTail {type=Transform, default {location=<0 0 0>, rotation=<rot 0 0 0>, scale=<1 1 1>}}
 Can Idle Break {type=boolean, default=true}
 Can Physics {type=boolean, default=true}
 ChestOpened {type=boolean}
 Current {type=number}
 Current_0 {type=number}
 dyn_tail_01 {type=Transform, default {location=<0 0 0>, rotation=<rot 0 0 0>, scale=<1 1 1>}}
 dyn_tail_02 {type=Transform, default {location=<0 0 0>, rotation=<rot 0 0 0>, scale=<1 1 1>}}
 dyn_tail_03 {type=Transform, default {location=<0 0 0>, rotation=<rot 0 0 0>, scale=<1 1 1>}}
 dyn_tail_04 {type=Transform, default {location=<0 0 0>, rotation=<rot 0 0 0>, scale=<1 1 1>}}
 dyn_tailfin_l_01 {type=Transform, default {location=<0 0 0>, rotation=<rot 0 0 0>, scale=<1 1 1>}}
 dyn_tailfin_l_02 {type=Transform, default {location=<0 0 0>, rotation=<rot 0 0 0>, scale=<1 1 1>}}
 dyn_tailfin_r_01 {type=Transform, default {location=<0 0 0>, rotation=<rot 0 0 0>, scale=<1 1 1>}}
 dyn_tailfin_r_02 {type=Transform, default {location=<0 0 0>, rotation=<rot 0 0 0>, scale=<1 1 1>}}
 EnemyPlayerEliminated {type=boolean}
 FirstReactiveDurationFE {type=number, default=30}
 Gimbal Rot {type=Rotator}
 Gravity Override {type=Vector}
 HealthDecreased {type=boolean}
 HealthIncreased {type=boolean}
 IdleAdditives {type=boolean}
 IdleBreak {type=boolean}
 IdleBreakDuration {type=number, default=0.2}
 IdleBreakEnableTime {type=number, default=0.2}
 IdleBreakLimitation {type=number, default=0.2}
 IdleBreakLimitationMax {type=number, default=30}
 IdleStop {type=boolean}
 Is Moving 2D {type=boolean}
 Physics {type=number}
 "Player  Skydiving" {type=boolean}
 Player Jump {type=boolean}
 PlayerCrouch {type=boolean}
 PlayerDBNO {type=boolean}
 PlayerSprint {type=boolean}
 Prev Look Direction {type=Vector2D, default=<0 0>}
 Prev Look Direction_0 {type=Vector2D, default=<0 0>}
 PrevLookDirection {type=Vector2D, default=<0 0>}
 PrevLowerLBlinkPose {type=number}
 PrevLowerRBlinkPose {type=number}
 PrevUpperLBlinkPose {type=number}
 PrevUpperRBlinkPose {type=number}
 ReactiveCount {type=number, default=0.2}
 ReactiveDurationFE {type=number, default=30}
 Rot Down {type=number}
 Rot Up {type=number}
 ScaleOffset {type=Vector, default=<1.1 1.2 1.2>}
 ShieldsDecreased {type=boolean}
 ShieldsIncreased {type=boolean}
 "Skel mesh " {type=SkeletalMeshComponent}
]
